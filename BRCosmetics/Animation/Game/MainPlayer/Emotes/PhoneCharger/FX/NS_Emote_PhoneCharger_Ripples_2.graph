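# Niagara system for the PhoneCharger emote that spawns ground-facing ripple sprites with greenish lights at the character's feet.
# Mirrored left and right foot emitters sample skeletal mesh bones, zeroing Emitter.Age until their System PuffCanSpawn bool is true to gate two quick bursts.
# Sprites grow, rotate and fade with curve-driven dynamic material parameters.
# Light specular is front-end only, and the left emitter also has a component renderer.
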NiagaraSystem /BRCosmetics/Animation/Game/MainPlayer/Emotes/PhoneCharger/FX/NS_Emote_PhoneCharger_Ripples_2
dependencies {
 usedDataInterfaces [&NiagaraDataInterfaceCurve, &NiagaraDataInterfaceSkeletalMesh]
 usedDynamicInputs [&SetBoolByFloatComparison, &MakeFloatFromVector]
 usedModules [
  &self:SystemSpawnScript.SystemScriptSource.SystemScriptGraph.NiagaraNodeAssignment_0.SetVariables_#1
  &SystemState
  &self:SystemSpawnScript.SystemScriptSource.SystemScriptGraph.NiagaraNodeAssignment_1.SetVariables_#2
 ]
}
emitters {
 Empty {
  emitterUpdateScript [
   &EmitterState(Life Cycle Mode=::NewEnumerator1, Loop Behavior=::NewEnumerator1, Loop Duration Mode=::NewEnumerator1, UseLoopDelay=false)
   /SetVariables(Emitter.Age:NiagaraFloat=&MakeCustomFloatFromBool(Bool=$System.PuffCanSpawnRightFoot, True Float=$Emitter.Age, False Float=0)) {isSetParameters=true, module=SetVariables_#3}
   &SpawnBurst_Instantaneous(Age:NiagaraFloat=$Emitter.Age, Spawn Time=0.08, Use Spawn Probability=true, UseLoopCountLimit=false)
   &SpawnBurst_Instantaneous(Age:NiagaraFloat=$Emitter.Age, Spawn Time=0.095, Use Spawn Probability=true, UseLoopCountLimit=false) {module=SpawnBurst_Instantaneous002}
  ]
  particleSpawnScript [
   &InitializeParticle(
    Color=<rgba 1 1 1 0.5>
    Sprite Rotation Mode=::NewEnumerator2
    Sprite Size Mode=::NewEnumerator3
    Uniform Sprite Size=&RandomRangeFloat(Minimum=55, Maximum=80)
    UsePositionOffset=false
    Write Material Random=true
   )
   &SampleSkeletalMesh(
    AdditionalYawPitchRoll=false
    Bone / Socket Index=1
    Bone Sampling Mode:ENiagaraBoneSamplingMode=::NewEnumerator4
    Sampled Position Offset:Vector3f=<0 0 -1>
    Skeletal Mesh:NiagaraDataInterfaceSkeletalMesh=$System.Skeletal Mesh
   )
  ]
  particleUpdateScript [
   &ParticleState()
   &DynamicMaterialParameters(
    Index 0 Param 1=&FloatFromCurve(
     FloatCurve=DataInterface() {propertyValues {Curve {keys [{arriveTangent=1.3374, interpMode=RCIM_Cubic, leaveTangent=1.3374, tangentMode=RCTM_User, time=0, value=1}]}}}
     CurveIndex=$Particles.Age
    )
    Index 0 Param 2=0.15
    Index 0 Param 3=&FloatFromCurve(
     FloatCurve=DataInterface() {propertyValues {Curve {keys [
         {arriveTangent=4.9025, interpMode=RCIM_Cubic, leaveTangent=4.9025, tangentMode=RCTM_User, time=0, value=0}
         {arriveTangent=4.4266, interpMode=RCIM_Cubic, tangentMode=RCTM_Break, time=0.5796, value=2}
        ]}}}
     CurveIndex=$Particles.Age
     Scale Curve=0.05
    )
    Index 0 Param 4=&FloatFromCurve(
     FloatCurve=DataInterface() {propertyValues {Curve {keys [
         {arriveTangent=-3.3445, interpMode=RCIM_Cubic, leaveTangent=-3.3445, tangentMode=RCTM_User, time=0, value=0.462}
         {arriveTangent=-1.1268, interpMode=RCIM_Cubic, leaveTangent=-1.2066, tangentMode=RCTM_Break, time=0.3394, value=-0.3632}
         {arriveTangent=-1.4939, interpMode=RCIM_Cubic, tangentMode=RCTM_Break, time=1.2, value=-1}
        ]}}}
     CurveIndex=$Particles.Age
    )
    Index 1 Param 1=0.8
    Index 1 Param 2=0.5
    Index 1 Param 4=3
    Module.Param0WriteEnabled=true
    Module.Param0WriteEnabled1=true
    Module.Param1WriteEnabled=true
    Module.Param1WriteEnabled1=true
    Module.Param2WriteEnabled=true
    Module.Param2WriteEnabled1=true
    Module.Param3WriteEnabled=true
    Module.Param3WriteEnabled1=true
    Write Parameter Index 1=true
   )
   &SpriteRotationRate(Rotation Rate=20)
   /SetVariables(Particles.SpriteFacing=&TransformVector(Vector=<0 0 1>, Source Space=::World, Destination Space=::Local)) {isSetParameters=true, module=SetVariables_#4}
   &ScaleSpriteSize(Uniform Curve Sprite Scale=DataInterface() {propertyValues {Curve {keys [
        {arriveTangent=4.1785, interpMode=RCIM_Cubic, leaveTangent=4.1785, tangentMode=RCTM_User, time=0, value=0}
        {arriveTangent=1.3778, interpMode=RCIM_Cubic, leaveTangent=1.3778, tangentMode=RCTM_User, time=0.1452, value=0.6}
        {arriveTangent=-0.0405, interpMode=RCIM_Cubic, leaveTangent=-0.0405, tangentMode=RCTM_User, time=1, value=1}
       ]}}})
   &ApplyOwnerScaleToAttributes(Scale Initial Sprite Size=true) {enabled=false}
   &SolveForcesAndVelocity(Clamp Velocity=false, Limit Acceleration=false)
   &ScaleColor(
    Scale Alpha=&FloatFromCurve(FloatCurve=DataInterface() {propertyValues {Curve {keys [<at 0.8 1>,<at 1 0>]}}}, Scale Curve=0.02)
    ScaleA=true
    ScaleRGB=true
   )
   /SetVariables(Particles.OffsetPosition:NiagaraPosition=&AddVectorToPosition(Position=$Particles.Position, Vector=$Particles.SampleSkeletalMesh.SampledPosition)) {isSetParameters=true, module=SetVariables_#5}
   &FN_Light_Attributes(
    Color=<rgba 0.4922 1 0.3904 1>
    Create Color Attribute=true
    Create Position Attribute=true
    Create Radius Attribute=true
    Position=&AddVectorToPosition(Position=<0 0 12>, Vector=Linked() {linkedVariable {name=Particles.OffsetPosition, type=NiagaraPosition}})
    Radius=40
    Scale Color=&FloatFromCurve(
     FloatCurve=DataInterface() {propertyValues {Curve {keys [{interpMode=RCIM_Cubic, leaveTangent=-2.5053, tangentMode=RCTM_Break, time=0, value=1}, <at 1 0>]}}}
     Scale Curve=3
    )
   )
   /SetVariables(Particles.LightSpecularScale:NiagaraFloat=&FN_FrontEndFloatModifier(Gameplay=0, Front End=1)) {isSetParameters=true}
  ]
  rendererClasses [&NiagaraSpriteRendererProperties, &NiagaraLightRendererProperties]
  renderers [
   {rendererClass=&NiagaraSpriteRendererProperties}
   {rendererClass=&NiagaraSpriteRendererProperties, rendererIndex=1}
   {rendererClass=&NiagaraLightRendererProperties, rendererIndex=2}
  ]
 }
 Empty001 {
  emitterUpdateScript [
   &EmitterState(Life Cycle Mode=::NewEnumerator1, Loop Behavior=::NewEnumerator1, Loop Duration Mode=::NewEnumerator1, UseLoopDelay=false)
   /SetVariables(Emitter.Age:NiagaraFloat=&MakeCustomFloatFromBool(Bool=$System.PuffCanSpawnLeftFoot, True Float=$Emitter.Age, False Float=0)) {isSetParameters=true, module=SetVariables_#3}
   &SpawnBurst_Instantaneous(Age:NiagaraFloat=$Emitter.Age, Spawn Time=0.08, Use Spawn Probability=true, UseLoopCountLimit=false)
   &SpawnBurst_Instantaneous(Age:NiagaraFloat=$Emitter.Age, Spawn Time=0.095, Use Spawn Probability=true, UseLoopCountLimit=false) {module=SpawnBurst_Instantaneous001}
  ]
  particleSpawnScript [
   &InitializeParticle(
    Color=<rgba 1 1 1 0.5>
    Sprite Rotation Mode=::NewEnumerator2
    Sprite Size Mode=::NewEnumerator3
    Uniform Sprite Size=&RandomRangeFloat(Minimum=55, Maximum=80)
    UsePositionOffset=false
    Write Material Random=true
   )
   &SampleSkeletalMesh(
    AdditionalYawPitchRoll=false
    Bone Sampling Mode:ENiagaraBoneSamplingMode=::NewEnumerator4
    Sampled Position Offset:Vector3f=<0 0 -1>
    Skeletal Mesh:NiagaraDataInterfaceSkeletalMesh=$System.Skeletal Mesh
   )
  ]
  particleUpdateScript [
   &ParticleState()
   &DynamicMaterialParameters(
    Index 0 Param 1=&FloatFromCurve(
     FloatCurve=DataInterface() {propertyValues {Curve {keys [{arriveTangent=1.3374, interpMode=RCIM_Cubic, leaveTangent=1.3374, tangentMode=RCTM_User, time=0, value=1}]}}}
     CurveIndex=$Particles.Age
    )
    Index 0 Param 2=0.15
    Index 0 Param 3=&FloatFromCurve(
     FloatCurve=DataInterface() {propertyValues {Curve {keys [
         {arriveTangent=4.9025, interpMode=RCIM_Cubic, leaveTangent=4.9025, tangentMode=RCTM_User, time=0, value=0}
         {arriveTangent=4.4266, interpMode=RCIM_Cubic, tangentMode=RCTM_Break, time=0.5796, value=2}
        ]}}}
     CurveIndex=$Particles.Age
     Scale Curve=0.05
    )
    Index 0 Param 4=&FloatFromCurve(
     FloatCurve=DataInterface() {propertyValues {Curve {keys [
         {arriveTangent=-3.3445, interpMode=RCIM_Cubic, leaveTangent=-3.3445, tangentMode=RCTM_User, time=0, value=0.462}
         {arriveTangent=-1.1268, interpMode=RCIM_Cubic, leaveTangent=-1.2066, tangentMode=RCTM_Break, time=0.3394, value=-0.3632}
         {arriveTangent=-1.4939, interpMode=RCIM_Cubic, tangentMode=RCTM_Break, time=1.2, value=-1}
        ]}}}
     CurveIndex=$Particles.Age
    )
    Index 1 Param 1=0.8
    Index 1 Param 2=0.5
    Index 1 Param 4=3
    Module.Param0WriteEnabled=true
    Module.Param0WriteEnabled1=true
    Module.Param1WriteEnabled=true
    Module.Param1WriteEnabled1=true
    Module.Param2WriteEnabled=true
    Module.Param2WriteEnabled1=true
    Module.Param3WriteEnabled=true
    Module.Param3WriteEnabled1=true
    Write Parameter Index 1=true
   )
   &SpriteRotationRate(Rotation Rate=20)
   /SetVariables(Particles.SpriteFacing=&TransformVector(Vector=<0 0 1>, Source Space=::World, Destination Space=::Local)) {isSetParameters=true, module=SetVariables_#4}
   &ScaleSpriteSize(Uniform Curve Sprite Scale=DataInterface() {propertyValues {Curve {keys [
        {arriveTangent=4.1785, interpMode=RCIM_Cubic, leaveTangent=4.1785, tangentMode=RCTM_User, time=0, value=0}
        {arriveTangent=1.3778, interpMode=RCIM_Cubic, leaveTangent=1.3778, tangentMode=RCTM_User, time=0.1452, value=0.6}
        {arriveTangent=-0.0405, interpMode=RCIM_Cubic, leaveTangent=-0.0405, tangentMode=RCTM_User, time=1, value=1}
       ]}}})
   &ApplyOwnerScaleToAttributes(Scale Initial Sprite Size=true) {enabled=false}
   &SolveForcesAndVelocity(Clamp Velocity=false, Limit Acceleration=false)
   &ScaleColor(
    Scale Alpha=&FloatFromCurve(
     FloatCurve=DataInterface() {propertyValues {Curve {keys [
         {interpMode=RCIM_Cubic, leaveTangent=-8.3587, tangentMode=RCTM_Break, time=0.8, value=1}
         {arriveTangent=-1.6308, interpMode=RCIM_Cubic, leaveTangent=-1.6308, tangentMode=RCTM_User, time=1, value=0}
        ]}}}
     Scale Curve=0.02
    )
    ScaleA=true
    ScaleRGB=true
   )
   /SetVariables(Particles.OffsetPosition:NiagaraPosition=&AddVectorToPosition(Position=$Particles.Position, Vector=$Particles.SampleSkeletalMesh.SampledPosition)) {isSetParameters=true, module=SetVariables_#5}
   &FN_Light_Attributes(
    Color=<rgba 0.4922 1 0.3904 1>
    Create Color Attribute=true
    Create Position Attribute=true
    Create Radius Attribute=true
    Position=&AddVectorToPosition(Position=<0 0 12>, Vector=Linked() {linkedVariable {name=Particles.OffsetPosition, type=NiagaraPosition}})
    Radius=40
    Scale Color=&FloatFromCurve(
     FloatCurve=DataInterface() {propertyValues {Curve {keys [{interpMode=RCIM_Cubic, leaveTangent=-2.5053, tangentMode=RCTM_Break, time=0, value=1}, <at 1 0>]}}}
     Scale Curve=3
    )
   )
   /SetVariables(Particles.LightSpecularScale:NiagaraFloat=&FN_FrontEndFloatModifier(Gameplay=0, Front End=1)) {isSetParameters=true}
  ]
  rendererClasses [
   &NiagaraSpriteRendererProperties
   &NiagaraLightRendererProperties
   &NiagaraComponentRendererProperties
  ]
  renderers [
   {rendererClass=&NiagaraSpriteRendererProperties}
   {rendererClass=&NiagaraSpriteRendererProperties, rendererIndex=1}
   {rendererClass=&NiagaraLightRendererProperties, rendererIndex=2}
   {rendererClass=&NiagaraComponentRendererProperties, rendererIndex=3}
  ]
 }
}
summary {
 systemName=NS_Emote_PhoneCharger_Ripples_2
 emitters [
  {
   emitterName=Empty001
   rendererClasses [
    &NiagaraSpriteRendererProperties
    &NiagaraLightRendererProperties
    &NiagaraComponentRendererProperties
   ]
  }
  {emitterName=Empty, rendererClasses [&NiagaraSpriteRendererProperties, &NiagaraLightRendererProperties]}
 ]
 userVariables [
  User.bIsFrontEnd {type=NiagaraBool}
  User.bIsFrontEndPreview {type=NiagaraBool}
 ]
}
