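# Niagara system for the PhoneCharger emote: emits small glowing green flares and sparkles that shrink away as they fall.
# Flares spawn at a steady rate in a small sphere, with curve-driven size, brightness and gravity, plus sprite and light renderers.
# Sparkles are hue-varied sprites that inherit velocity and fall with drag.
# Despite the UnitSpawning name, SpawnPerUnit is disabled in favor of SpawnRate, as is skeletal mesh placement in both emitters.
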NiagaraSystem /BRCosmetics/Animation/Game/MainPlayer/Emotes/PhoneCharger/FX/NS_Emote_PhoneCharger_UnitSpawning
dependencies {
 usedDataInterfaces [&NiagaraDataInterfaceCurve, &NiagaraDataInterfaceSkeletalMesh]
 usedDynamicInputs []
 usedModules [
  &self:SystemSpawnScript.SystemScriptSource.SystemScriptGraph.NiagaraNodeAssignment_0.SetVariables_#1
  &SystemState
 ]
}
emitters {
 Flares {
  emitterUpdateScript [
   &EmitterState()
   &SpawnRate(SpawnRate=8, Use Spawn Probability=false)
  ]
  particleSpawnScript [
   &InitializeParticle(
    Color=<rgba 7.2897 15 1.875 0.8>
    Lifetime Max=0.6
    Lifetime Min=0.35
    Lifetime Mode=::NewEnumerator1
    Sprite Size Mode=::NewEnumerator1
    Uniform Sprite Size Max=20
    Uniform Sprite Size Min=10
    UsePositionOffset=false
    Write Material Random=true
   )
   &PartitionParticles()
   &ShapeLocation(MaskSpawnGroup=false, Override Seed=false, Sphere Radius=12, UseOwnerScale=true)
   &SkeletalMeshLocation(
    Bone / Socket Index=$Particles.PartitionParticles.Partition
    Bone Sampling Mode=::NewEnumerator4
    MaskSpawnGroup=false
    Sampled Position Offset=$Particles.ShapeLocation.ShapePosition
    Skeletal Mesh=$System.Skeletal Mesh
   ) {enabled=false}
   &DynamicMaterialParameters(
    Index 0 Param 1=&RandomRangeFloat(Minimum=0.2, Maximum=0.4)
    Index 0 Param 2=1
    Module.Index0Param0Write=true
    Module.Index0Param1Write=true
    Module.Index0Param2Write=true
    Module.Index0Param3Write=true
   )
  ]
  particleUpdateScript [
   &ParticleState()
   &ScaleSpriteSize(Uniform Curve Sprite Scale=DataInterface() {propertyValues {Curve {keys [
        {tangentMode=RCTM_User, time=0, value=0.05}
        <at 0.2024 0.4>
        <at 0.291 1.45>
        {interpMode=RCIM_Cubic, tangentMode=RCTM_Break, time=0.3954, value=0.7}
        {arriveTangent=-2.0701, interpMode=RCIM_Cubic, tangentMode=RCTM_Break, time=1, value=0}
       ]}}})
   &GravityForce(Gravity=&Multiply_VectorByFloat(
     Vector=<1 1 -100>
     Float=&FloatFromCurve(FloatCurve=DataInterface() {propertyValues {Curve {keys [
          {interpMode=RCIM_Cubic, leaveTangent=1.2256, tangentMode=RCTM_Break, time=0, value=0.05}
          {arriveTangent=1.148, tangentMode=RCTM_User, time=1, value=0.85}
         ]}}})
    ))
   &SolveForcesAndVelocity(Clamp Velocity=false, Limit Acceleration=false)
   &ScaleColor(
    Scale RGB=&VectorFromFloat(Value=&FloatFromCurve(FloatCurve=DataInterface() {propertyValues {Curve {keys [
          {interpMode=RCIM_Cubic, leaveTangent=-9.1901, tangentMode=RCTM_Break, time=0, value=1}
          {arriveTangent=-0.7291, interpMode=RCIM_Cubic, leaveTangent=-0.7291, tangentMode=RCTM_User, time=0.12, value=0.3}
          <at 1 0>
         ]}}}))
    ScaleA=true
    ScaleRGB=true
   )
   &FN_Light_Attributes(
    Color=$Particles.Color
    Create Color Attribute=true
    Create Exponent Attribute=true
    Create Radius Attribute=true
    Exponent=2
    Radius=45
    Scale Color=0.5
   )
  ]
  rendererClasses [&NiagaraSpriteRendererProperties, &NiagaraLightRendererProperties]
  renderers [
   {rendererClass=&NiagaraSpriteRendererProperties}
   {rendererClass=&NiagaraLightRendererProperties, rendererIndex=1}
  ]
 }
 Sparkles {
  emitterUpdateScript [
   &EmitterState()
   &SpawnPerUnit(Module.Use Max Movement Threshold=true, Module.Use Movement Tolerance=false, Use Spawn Probability=false) {enabled=false}
   &SpawnRate(SpawnRate=20, Use Spawn Probability=false)
  ]
  particleSpawnScript [
   &InitializeParticle(
    AdjustAlpha=false
    AdjustHue=true
    AdjustSaturation=false
    AdjustValue=false
    Color=<rgba 0.972 2 0.25 0.5>
    Color Mode=::NewEnumerator3
    Lifetime Max=0.6
    Lifetime Min=0.3
    Lifetime Mode=::NewEnumerator1
    Sprite Size Mode=::NewEnumerator1
    Uniform Sprite Size Max=35
    Uniform Sprite Size Min=10
    UsePositionOffset=false
    Write Material Random=true
   )
   &PartitionParticles()
   &ShapeLocation(MaskSpawnGroup=false, Override Seed=false, Sphere Radius=5, UseOwnerScale=true)
   &InheritVelocity(Speed Limit=true, UseSpeedThreshold=false)
   &SkeletalMeshLocation(
    Bone / Socket Index=$Particles.PartitionParticles.Partition
    Bone Sampling Mode=::NewEnumerator4
    MaskSpawnGroup=false
    Sampled Position Offset=$Particles.ShapeLocation.ShapePosition
    Skeletal Mesh=$System.Skeletal Mesh
   ) {enabled=false}
   &DynamicMaterialParameters(
    Index 0 Param 1=&RandomRangeFloat(Minimum=0.3, Maximum=0)
    Index 0 Param 2=&RandomRangeFloat(Minimum=1, Maximum=2)
    Index 0 Param 3=&RandomRangeFloat(Minimum=-1, Maximum=1)
    Module.Index0Param0Write=true
    Module.Index0Param1Write=true
    Module.Index0Param2Write=true
    Module.Index0Param3Write=true
   )
  ]
  particleUpdateScript [
   &ParticleState()
   &ScaleSpriteSize(Uniform Curve Sprite Scale=DataInterface() {propertyValues {Curve {keys [{tangentMode=RCTM_User, time=0, value=0.05}, <at 0.2 0.85>, <at 0.5 0.95>, {tangentMode=RCTM_User, time=1, value=0}]}}})
   &GravityForce(Gravity=<0 0 -200>)
   &Drag(Drag=2.5, Use Linear Drag=true, Use Rotational Drag=true)
   &SolveForcesAndVelocity(Clamp Velocity=false, Limit Acceleration=false)
  ]
  rendererClasses [&NiagaraSpriteRendererProperties]
  renderers [{rendererClass=&NiagaraSpriteRendererProperties}]
 }
}
summary {
 systemName=NS_Emote_PhoneCharger_UnitSpawning
 emitters [
  {emitterName=Sparkles, rendererClasses [&NiagaraSpriteRendererProperties]}
  {emitterName=Flares, rendererClasses [&NiagaraSpriteRendererProperties, &NiagaraLightRendererProperties]}
 ]
 userVariables []
}
